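# Animation Blueprint for the StylusFluff backpack that repositions its pack bone to follow the wearer's poses and movement states.
# It blends copied parent and BodyMesh poses per bone, stacks additive pack offsets, runs a Control Rig, then caches the result.
# Event graph derives offset alphas from bone deltas, sideways velocity, and melee, skydiving and parachute flags.
# A final bool blend uses the cached pose directly in the front end, otherwise adds another small offset.
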AnimBlueprint /BRCosmetics/Accessories/FORT_Backpacks/Backpack_M_MED_StylusFluff/Meshes/M_MED_StylusFluff_Pack_AnimBP
target_skeleton=&/M_MED_StylusFluff_Pack_Skeleton
graphs [
 AnimGraph {nodes [
   Input Pose()
   n1 = "Transform (Modify) Bone - Bone: pack"(ComponentPose=Local To Component(LocalPose=Use cached pose 'Cache of all Caches'()), Translation="0,-2,0", Rotation="0,0,0") {
    rotationMode=BMM_Additive
    translationMode=BMM_Additive
    alphaBoolBlend {blendOption=Linear}
    boneToModify {boneName=pack}
    translation {y=-2}
   }
   Output Pose(Result=Blend Poses by bool(BlendPose_0=Component To Local(ComponentPose=@n1), BlendPose_1=Use cached pose 'Cache of all Caches'(), bActiveValue=Get bIsInFrontEnd()) {blendType=Linear}) {layerGroup=None}
   n2 = Layered blend per bone(
    BasePose=Copy Pose From Mesh()
    BlendPoses_0=Copy Pose From Mesh(SourceMeshComponent=Get BodyMesh() -> (BodyMesh:Skeletal Mesh Component Object Reference))
   ) {layerSetup [{branchFilters [{boneName=spine_05}, {boneName=clavicle_l}, {boneName=clavicle_r}, {boneName=upperarm_l}, {boneName=upperarm_r}]}]}
   n3 = "Transform (Modify) Bone - Bone: pack"(
    ComponentPose=Local To Component(LocalPose=@n2)
    Alpha=Get LeftClavPositionAlpha() -> (LeftClavPositionAlpha:double)
    Translation="-0.5,-3,0.5"
    Rotation="-0.000,-5.000,0.000"
   ) {
    rotationMode=BMM_Additive
    rotationSpace=BCS_BoneSpace
    translationMode=BMM_Additive
    translationSpace=BCS_BoneSpace
    alphaBoolBlend {blendInTime=0.3, blendOutTime=1}
    boneToModify {boneName=pack}
    rotation {yaw=-5}
    translation=<-0.5 -3 0.5>
   }
   n4 = "Transform (Modify) Bone - Bone: pack"(ComponentPose=@n3, Alpha=Get SpineRotationAlpha() -> (SpineRotationAlpha:double), Translation="0,-3,0", Rotation="1.036,-9.904,20") {
    rotationMode=BMM_Additive
    rotationSpace=BCS_BoneSpace
    translationMode=BMM_Additive
    translationSpace=BCS_BoneSpace
    alphaBoolBlend {blendInTime=0.3, blendOutTime=1}
    boneToModify {boneName=pack}
    rotation=<rot 1.036 -9.904 20>
    translation {y=-3}
   }
   n5 = "Transform (Modify) Bone - Bone: pack"(ComponentPose=@n4, Alpha=Get Skydiving_to_Left() -> (Skydiving_to_Left:double), Translation="1,-1,0", Rotation="0,5,8") {
    rotationMode=BMM_Additive
    rotationSpace=BCS_BoneSpace
    translationMode=BMM_Additive
    translationSpace=BCS_BoneSpace
    alphaBoolBlend {blendInTime=0.3, blendOutTime=1}
    boneToModify {boneName=pack}
    rotation {roll=8, yaw=5}
    translation=<1 -1>
   }
   n6 = "Transform (Modify) Bone - Bone: pack"(ComponentPose=@n5, Alpha=Get Skydiving_to_Right() -> (Skydiving_to_Right:double), Translation="-3,-1,0", Rotation="-0.000,-10.000,0.000") {
    rotationMode=BMM_Additive
    rotationSpace=BCS_BoneSpace
    translationMode=BMM_Additive
    translationSpace=BCS_BoneSpace
    alphaBoolBlend {blendInTime=0.3, blendOutTime=1}
    boneToModify {boneName=pack}
    rotation {yaw=-10}
    translation=<-3 -1>
   }
   n7 = "Transform (Modify) Bone - Bone: pack"!(ComponentPose=@n6, bAlphaBoolEnabled:bool=Get bIsSkyDiving(), Translation="0,-5,1", Rotation="0,0,13", Scale) -> (Pose) {
    alphaInputType=Bool
    rotationMode=BMM_Additive
    rotationSpace=BCS_BoneSpace
    translationMode=BMM_Additive
    translationSpace=BCS_BoneSpace
    alphaBoolBlend {blendInTime=0.3, blendOption=ExpInOut, blendOutTime=0.3}
    boneToModify {boneName=pack}
    rotation {roll=13}
    translation {y=-5, z=1}
   }
   n8 = "Transform (Modify) Bone - Bone: pack"!(
    ComponentPose=@n7
    Alpha:float=Get SpineCurvatureCompensationAlpha() -> (SpineCurvatureCompensationAlpha:double)
    Translation="1.836,-11.484,17.962"
    Rotation="0.000,0.956,-0.366"
    Scale
   ) -> (Pose) {
    rotationMode=BMM_Additive
    rotationSpace=BCS_BoneSpace
    translationMode=BMM_Additive
    translationSpace=BCS_BoneSpace
    alphaBoolBlend {blendInTime=0.3, blendOutTime=1}
    boneToModify {boneName=pack}
    rotation=<rot 0 0.956 -0.3661>
    translation=<1.836 -11.484 17.962>
   }
   n9 = Control Rig(Source=Component To Local(ComponentPose=@n8)) {
    alphaBoolBlend {blendOption=Linear}
    controlRigAssetReference {blueprintRigClass=&/M_MED_StylusFluff_Pack_CtrlRig.M_MED_StylusFluff_Pack_CtrlRig_C}
   }
   Save cached pose 'Cache of all Caches'(Pose=@n9) {cacheName=Cache of all Caches}
  ]}
 EventGraph {nodes [
   Event BlueprintUpdateAnimation()
   > n1 = Sequence!(execute) -> (then_0, then_1, then_2:Exec, then_3:Exec, then_4:Exec, then_5:Exec)
   n2 = TryGetPawnOwner()
   n3 = Dot Product!(
    A:Vector=GetVelocity!(self:Actor Object Reference=@n2) -> (ReturnValue:Vector)
    B:Vector=GetActorRightVector!(self:Actor Object Reference=@n2) -> (ReturnValue:Vector)
   ) -> (ReturnValue:double)
   n4 = Set IsGoingRightAlpha(execute=@n1.then_2, IsGoingRightAlpha:double=MapRangeClamped(Value=@n3, InRangeA=0, InRangeB=400)) -> (Output_Get:double)
   > Set Skydiving_to_Right(Skydiving_to_Right:double=float * float(A=@n4, B:bool=Get bIsSkyDiving()) {is=title K2Node_PromotableOperator}) -> (Output_Get:double)
   n5 = Set IsGoingLeftAlpha(execute=@n1.then_3, IsGoingLeftAlpha:double=MapRangeClamped(Value=@n3, InRangeA=0, InRangeB=-400)) -> (Output_Get:double)
   > Set Skydiving_to_Left(Skydiving_to_Left:double=float * float(A=@n5, B=Get bIsSkyDiving())) -> (Output_Get:double)
   n6 = GetOwnerSkeletalMeshComponent()
   Set SpineCurvatureCompensationAlpha(
    execute=@n1.then_4
    SpineCurvatureCompensationAlpha:double=float * float(
     A=GetAlphaFromReferenceTransformDelta(SkeletalMeshComponent=@n6, BoneName=spine_03, BaseName=spine_01, InputRange=(X=-7,Y=-20), TwistAxis="1,0,0")
     B=Get bIsParachuteOpened()
    )
   ) -> (Output_Get:double)
   n7 = GetOwnerSkeletalMeshComponent()
   n8 = GetDeltaTransformFromRefPose(self=@n7, BoneName=spine_05, BaseName=clavicle_l)
   Set LeftClavPositionAlpha(
    execute=@n1.then_0
    LeftClavPositionAlpha:double=MapRangeClamped(
     Value=BreakTransform!(InTransform=@n8) -> (Location, Rotation_Roll:float, Rotation_Pitch:float, Rotation_Yaw:float, Scale).Rotation_Roll
     InRangeA=-8
     InRangeB=-20
    )
   ) -> (Output_Get:double)
   n9 = BreakTransform(InTransform=GetDeltaTransformFromRefPose(self=@n7, BoneName=pelvis, BaseName=spine_05))
   Set SpineRotationAlpha(
    execute=@n1.then_1
    SpineRotationAlpha:double=float * float(A=MapRangeClamped(Value=@n9.Rotation_Yaw, InRangeA=0, InRangeB=15), B=Get bIsPlayingMeleeAnim())
   ) -> (Output_Get:double)
   Event OnCharacterCustomizationLoaded()
   > n10 = Cast To PlayerPawn_Athena(Object=@n2)
   > Set BodyMesh(BodyMesh=GetSkeletalMeshForPartType(self=@n10))
  ]}
]
variables [
 BodyMesh {type=SkeletalMeshComponent}
 IsGoingLeftAlpha {type=number}
 IsGoingRightAlpha {type=number}
 LeftClavPositionAlpha {type=number}
 Skydiving_to_Left {type=number}
 Skydiving_to_Right {type=number}
 SpineCurvatureCompensationAlpha {type=number}
 SpineRotationAlpha {type=number}
]
